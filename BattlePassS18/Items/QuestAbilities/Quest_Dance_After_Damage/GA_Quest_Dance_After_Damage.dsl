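; Battle Pass S18 quest gameplay ability that, going by its name, tracks whether the player dances after a damage event.
; On event activation it ends if target tags match the prey container, otherwise starts a looping timer.
; The timer callback checks the player pawn for the dance tag, accumulating dancing or waiting time.
; Enough dancing sends a quest stat event; timer clearing and ability end follow the success and timeout checks.
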(event Ability|EventActivateAbilityFromEvent
  (params (EventData :type Gameplay%20Event%20Data%20Structure%20%28by%20ref%29))
  (bind (_eventtag _instigator _target _optionalobject _optionalobject2 _contexthandle _instigatortags _targettags) (Utilities|Struct|BreakGameplayEventData EventData))
  (|CallParentFunction EventData)
  (Variables|Default|SetTimeDancing 0.0)
  (Variables|Default|SetTimeChecking 0.0)
  (Development|PrintString "@@@ ActivateAbilityFromEvent [GA_Quest_Dance_After_Damage]: CALLED!" false (Variables|Debug|GetbDebug))
  (if (GameplayTags|HasAnyTags _targettags (Variables|Default|GetPreyContainer))
    (Ability|EndAbility)
    (else
      (bind _returnvalue (Utilities|Time|SetTimerbyEvent (delegate Custom|CheckForDance) (Variables|Default|GetCheckInterval) true))
      (Variables|Default|SetTimer_CheckForDance _returnvalue))))

(event Custom|CheckForDance
  (bind _returnvalue (MissionHelpers|HasTags (GameplayTags|GetOwnedGameplayTags (Variables|GAVars|Data|GetOwningPlayerPawn) :!class /Script/GameplayTags.GameplayTagAssetInterface) (Variables|Default|GetDanceTag) false false))
  (if _returnvalue
    (bind _output_get (Variables|Default|SetTimeDancing (+ (Variables|Default|GetTimeDancing) (Variables|Default|GetCheckInterval))))
    (Development|PrintString (Utilities|String|Append "@@@@@@ ActivateAbilityFromEvent [GA_Quest_Dance_After_Damage]: Dancing for " (Utilities|String|ToString(Float) _output_get) "seconds...") false (Variables|Debug|GetbDebug))
    (if (>= _output_get (Variables|Default|GetTimeToDance))
      (Quest|SendComplexCustomStatEvent (Variables|GAVars|Data|GetOwningPlayerQuestManager) 0 (Variables|Default|GetQuestTag) "(GameplayTags=)" 1)
      (Development|PrintString "@@@@@@@@@ ActivateAbilityFromEvent [GA_Quest_Dance_After_Damage]: SUCCESS!" false (Variables|Debug|GetbDebug)))
    (else
      (bind _output_get_2 (Variables|Default|SetTimeChecking (+ (Variables|Default|GetTimeChecking) (Variables|Default|GetCheckInterval))))
      (if (>= _output_get_2 (Variables|Default|GetTimeToCheck))
        (Development|PrintString "@@@@@@@@@ ActivateAbilityFromEvent [GA_Quest_Dance_After_Damage]: TIMEOUT!" false (Variables|Debug|GetbDebug)))))
  (Utilities|Time|ClearandInvalidateTimerbyHandle (Variables|Default|GetTimer_CheckForDance))
  (Ability|EndAbility))
